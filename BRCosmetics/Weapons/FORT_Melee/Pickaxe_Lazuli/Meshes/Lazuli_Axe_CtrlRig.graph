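# Control Rig for the Lazuli two-handed pickaxe that procedurally animates the three wings (lower, middle, upper) on the axe.
# The execution chain runs three ChainHarmonicsPerItem units in sequence, each rooted at one wing's base bone under the weapon handle.
# Each layers a sine-eased wave with pendulum lag, staggering wave offsets and ranges per wing so they move out of sync.
# Debug drawing is left enabled on all three units.
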ControlRigBlueprint /BRCosmetics/Weapons/FORT_Melee/Pickaxe_Lazuli/Meshes/Lazuli_Axe_CtrlRig
graphs [
 RigVMFunctionLibrary {links [], nodes []}
 RigVMModel {nodes [
   RigUnit_BeginExecution = &RigUnit_BeginExecution()
   /Script/ControlRig.RigUnit_ChainHarmonicsPerItem(
    ChainRoot=(Type=Bone,Name="dyn_wing_lower_base_01")
    DrawWorldOffset=(Rotation=(X=0,Y=0,Z=0,W=1),Translation=(X=0,Y=0,Z=0),Scale3D=(X=1,Y=1,Z=1))
    Pendulum=(bEnabled=True,PendulumStiffness=2,PendulumGravity=(X=0,Y=0,Z=0),PendulumBlend=0.75,PendulumDrag=0.98,PendulumMinimum=0,PendulumMaximum=0.1,PendulumEase=Linear,UnwindAxis=(X=0,Y=1,Z=0),UnwindMinimum=0.2,UnwindMaximum=0.05)
    Reach=(bEnabled=True,ReachTarget=(X=0,Y=0,Z=0),ReachAxis=(X=1,Y=0,Z=0),ReachMinimum=0,ReachMaximum=0,ReachEase=Linear)
    Speed=(X=1,Y=1.7,Z=1)
    Wave=(bEnabled=True,WaveFrequency=(X=1,Y=0.6,Z=1),WaveAmplitude=(X=0,Y=0.5,Z=0.3),WaveOffset=(X=0,Y=0.6,Z=-1),WaveNoise=(X=0,Y=0,Z=0),WaveMinimum=-0.5,WaveMaximum=0.1,WaveEase=SineEaseInOut)
    WaveCurve="(EditorCurveData=(Keys=((InterpMode=RCIM_Linear,TangentMode=RCTM_Auto,TangentWeightMode=RCTWM_WeightedNone,Time=-0.0109,Value=0.5511,ArriveTangent=0,ArriveTangentWeight=0,LeaveTangent=0,LeaveTangentWeight=0),(InterpMode=RCIM_Linear,TangentMode=RCTM_Auto,Ta ...(truncated)"
    bDrawDebug=True
    ExecutePin=@RigUnit_BeginExecution.ExecutePin
   ) {is=struct RigVMUnitNode}
   > &RigUnit_ChainHarmonicsPerItem(
    ChainRoot=(Type=Bone,Name="dyn_wing_middle_base_01")
    DrawWorldOffset=(Rotation=(X=0,Y=0,Z=0,W=1),Translation=(X=0,Y=0,Z=0),Scale3D=(X=1,Y=1,Z=1))
    Pendulum=(bEnabled=True,PendulumStiffness=2,PendulumGravity=(X=0,Y=0,Z=0),PendulumBlend=0.75,PendulumDrag=0.98,PendulumMinimum=0,PendulumMaximum=0.1,PendulumEase=Linear,UnwindAxis=(X=0,Y=1,Z=0),UnwindMinimum=0.2,UnwindMaximum=0.05)
    Reach=(bEnabled=True,ReachTarget=(X=0,Y=0,Z=0),ReachAxis=(X=1,Y=0,Z=0),ReachMinimum=0,ReachMaximum=0,ReachEase=Linear)
    Speed=(X=1,Y=1.7,Z=1)
    Wave=(bEnabled=True,WaveFrequency=(X=1,Y=0.6,Z=1),WaveAmplitude=(X=0,Y=0.5,Z=0.3),WaveOffset=(X=0,Y=1,Z=-0.5),WaveNoise=(X=0,Y=0,Z=0),WaveMinimum=-0.6,WaveMaximum=0.15,WaveEase=SineEaseInOut)
    WaveCurve="(EditorCurveData=(Keys=((InterpMode=RCIM_Linear,TangentMode=RCTM_Auto,TangentWeightMode=RCTWM_WeightedNone,Time=-0.0195,Value=0.5223,ArriveTangent=0,ArriveTangentWeight=0,LeaveTangent=0,LeaveTangentWeight=0),(InterpMode=RCIM_Linear,TangentMode=RCTM_Auto,Ta ...(truncated)"
    bDrawDebug=True
   )
   > &RigUnit_ChainHarmonicsPerItem(
    ChainRoot=(Type=Bone,Name="dyn_wing_upper_base_01")
    DrawWorldOffset=(Rotation=(X=0,Y=0,Z=0,W=1),Translation=(X=0,Y=0,Z=0),Scale3D=(X=1,Y=1,Z=1))
    Pendulum=(bEnabled=True,PendulumStiffness=2,PendulumGravity=(X=0,Y=0,Z=0),PendulumBlend=0.75,PendulumDrag=0.98,PendulumMinimum=0,PendulumMaximum=0.1,PendulumEase=Linear,UnwindAxis=(X=0,Y=1,Z=0),UnwindMinimum=0.2,UnwindMaximum=0.05)
    Reach=(bEnabled=True,ReachTarget=(X=0,Y=0,Z=0),ReachAxis=(X=1,Y=0,Z=0),ReachMinimum=0,ReachMaximum=0,ReachEase=Linear)
    Speed=(X=1,Y=1.7,Z=1)
    Wave=(bEnabled=True,WaveFrequency=(X=1,Y=0.6,Z=1),WaveAmplitude=(X=0,Y=0.5,Z=0.3),WaveOffset=(X=0,Y=1.5,Z=0),WaveNoise=(X=0,Y=0,Z=0),WaveMinimum=-0.65,WaveMaximum=0.15,WaveEase=SineEaseInOut)
    WaveCurve="(EditorCurveData=(Keys=((InterpMode=RCIM_Linear,TangentMode=RCTM_Auto,TangentWeightMode=RCTWM_WeightedNone,Time=-0.0053,Value=0.5092,ArriveTangent=0,ArriveTangentWeight=0,LeaveTangent=0,LeaveTangentWeight=0),(InterpMode=RCIM_Linear,TangentMode=RCTM_Auto,Ta ...(truncated)"
    bDrawDebug=True
   )
  ]}
]
hierarchy [
 b_AR_Root {loc=<0 0 0>, rot=<0 0 0.7071 0.7071>}
 Melee_TwoHanded_Handle {parent=b_AR_Root, loc=<8.1477 -0.008 -29.0793>, rot=<0 0 0.7071 -0.7071>}
 dyn_wing_lower_base_01 {parent=Melee_TwoHanded_Handle, loc=<0.7893 11.837 107.1118>, rot=<-0.1395 -0.1395 0.6932 -0.6932>}
 dyn_wing_lower_01 {parent=dyn_wing_lower_base_01, loc=<15.8413 0 0>, rot=<0 -0.1338 0 -0.991>}
 dyn_wing_middle_base_01 {parent=Melee_TwoHanded_Handle, loc=<0.7893 3.3034 113.4386>, rot=<-0.1385 -0.1385 0.6934 -0.6934>}
 dyn_wing_middle_01 {parent=dyn_wing_middle_base_01, loc=<15.6954 0 0>, rot=<0 -0.0297 0 -0.9996>}
 dyn_wing_upper_base_01 {parent=Melee_TwoHanded_Handle, loc=<0.7893 -0.6811 119.0529>, rot=<-0.0585 -0.0576 0.7048 -0.7047>}
 dyn_wing_upper_01 {parent=dyn_wing_upper_base_01, loc=<18.7601 0 0>, rot=<0.0007 -0.0973 -0.0005 -0.9953>}
]
