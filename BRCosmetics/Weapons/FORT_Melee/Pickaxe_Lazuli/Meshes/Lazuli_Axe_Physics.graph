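# Physics asset for the Lazuli pickaxe mesh, giving collision bodies and joints to the handle and its three dynamic wing chains.
# A large sphere covers the two-handed handle bone, while small spheres sit on the base and tip bones of the upper, middle and lower wings.
# Each tip is constrained to its base and each base to the handle, all sharing the same narrow swing and twist limits.
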PhysicsAsset /BRCosmetics/Weapons/FORT_Melee/Pickaxe_Lazuli/Meshes/Lazuli_Axe_Physics
bodies {
 Melee_TwoHanded_Handle [{shapeType=Sphere, radius=26.4377, shapeName=Melee_TwoHanded_Handle_sphere, center=<0.7893 2.064 31.642>}]
 dyn_wing_lower_01 [{shapeType=Sphere, radius=5, shapeName=0, center=<1 0 0>}]
 dyn_wing_lower_base_01 [{shapeType=Sphere, radius=5, shapeName=0, center=<1 0 0>}]
 dyn_wing_middle_01 [{shapeType=Sphere, radius=7, shapeName=0, center=<3.5 0 0>}]
 dyn_wing_middle_base_01 [{shapeType=Sphere, radius=7, shapeName=0, center=<3.5 0 0>}]
 dyn_wing_upper_01 [{shapeType=Sphere, radius=5, shapeName=0, center=<3 0 0>}]
 dyn_wing_upper_base_01 [{shapeType=Sphere, radius=5, shapeName=0, center=<3 0 0>}]
}
constraints [
 {bone1Name=dyn_wing_lower_01, bone2Name=dyn_wing_lower_base_01, swing1LimitDegrees=5, swing2LimitDegrees=16, twistLimitDegrees=5}
 {bone1Name=dyn_wing_middle_01, bone2Name=dyn_wing_middle_base_01, swing1LimitDegrees=5, swing2LimitDegrees=16, twistLimitDegrees=5}
 {bone1Name=dyn_wing_upper_01, bone2Name=dyn_wing_upper_base_01, swing1LimitDegrees=5, swing2LimitDegrees=16, twistLimitDegrees=5}
 {bone1Name=dyn_wing_middle_base_01, bone2Name=Melee_TwoHanded_Handle, swing1LimitDegrees=5, swing2LimitDegrees=16, twistLimitDegrees=5}
 {bone1Name=dyn_wing_upper_base_01, bone2Name=Melee_TwoHanded_Handle, swing1LimitDegrees=5, swing2LimitDegrees=16, twistLimitDegrees=5}
 {bone1Name=dyn_wing_lower_base_01, bone2Name=Melee_TwoHanded_Handle, swing1LimitDegrees=5, swing2LimitDegrees=16, twistLimitDegrees=5}
]
