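# Animation Blueprint for the Lazuli axe pickaxe, adding procedural motion to its wing bones.
# A Control Rig pose, faded out while moving or airborne, is cached and fed to RigidBody, an AnimDynamics chain, or straight through, chosen by bools.
# The AnimDynamics chain first pitches wing bases additively while jumping, falling or sliding, then simulates springy wing bones.
# EventGraph sets alphas from movement and melee state; its FInterpTo follows a direct set, so has no easing effect.
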AnimBlueprint /BRCosmetics/Weapons/FORT_Melee/Pickaxe_Lazuli/Meshes/Lazuli_Axe_AnimBP
target_skeleton=&/Lazuli_Axe_Skeleton
graphs [
 AnimGraph {nodes [
   n1 = Get AnimDynamicsAlpha() -> (AnimDynamicsAlpha:double)
   n2 = Get WingTransformWhileJumpingAlpha() -> (WingTransformWhileJumpingAlpha:double)
   n3 = "Transform (Modify) Bone - Bone: dyn_wing_lower_base_01"(ComponentPose=Local To Component(LocalPose=Use cached pose 'Post-CR'()), Alpha=@n2, Translation="0,0,0", Rotation="3,0,0", Scale="1,1,1") {
    rotationMode=BMM_Additive
    rotationSpace=BCS_BoneSpace
    alphaBoolBlend {blendOption=Linear}
    alphaScaleBiasClamp {bInterpResult=true}
    boneToModify {boneName=dyn_wing_lower_base_01}
    rotation {pitch=3}
   }
   n4 = "Transform (Modify) Bone - Bone: dyn_wing_upper_base_01"(ComponentPose=@n3, Alpha=@n2, Translation="0,0,0", Rotation="8,0,0", Scale="1,1,1") {
    rotationMode=BMM_Additive
    rotationSpace=BCS_BoneSpace
    alphaBoolBlend {blendOption=Linear}
    alphaScaleBiasClamp {bInterpResult=true}
    boneToModify {boneName=dyn_wing_upper_base_01}
    rotation {pitch=8}
   }
   n5 = "Transform (Modify) Bone - Bone: dyn_wing_middle_base_01"(ComponentPose=@n4, Alpha=@n2, Translation="0,0,0", Rotation="8,0,0", Scale="1,1,1") {
    rotationMode=BMM_Additive
    rotationSpace=BCS_BoneSpace
    alphaBoolBlend {blendOption=Linear}
    alphaScaleBiasClamp {bInterpResult=true}
    boneToModify {boneName=dyn_wing_middle_base_01}
    rotation {pitch=8}
   }
   n6 = Anim Dynamics - Bone: dyn_wing_upper_base_01(ComponentPose=@n5, Alpha=@n1, ExternalForce="0,0,0") {
    angularBiasOverride=1
    angularDampingOverride=1
    angularSpringConstant=10000
    bAngularSpring=true
    bOverrideAngularBias=true
    bOverrideAngularDamping=true
    bUseGravityOverride=true
    linearSpringConstant=10000
    simulationSpace=World
    alphaBoolBlend {blendOption=Linear}
    alphaScaleBiasClamp {bInterpResult=true, interpSpeedDecreasing=20, interpSpeedIncreasing=20}
    boundBone {boneName=dyn_wing_upper_base_01}
    componentAppliedLinearAccClamp=<100 100 100>
    componentLinearAccScale=<0.1 0.1 0.1>
    componentLinearVelScale=<0.2 0.2 0.2>
    gravityOverride {z=-600}
    physicsBodyDefinitions [{
      boundBone {boneName=dyn_wing_upper_base_01}
      boxExtents=<30 40 40>
      constraintSetup {angularLimitsMax=<0 8 3>, angularLimitsMin=<0 -10 -3>, angularTarget=<0 0 0>}
      localJointOffset=<30 0 0>
     }]
    relativeSpaceBone {boneName=Melee_TwoHanded_Handle}
   }
   n7 = Anim Dynamics - Bone: dyn_wing_middle_base_01(ComponentPose=@n6, Alpha=@n1, ExternalForce="0,0,0") {
    angularBiasOverride=1
    angularDampingOverride=1
    angularSpringConstant=10000
    bAngularSpring=true
    bOverrideAngularBias=true
    bOverrideAngularDamping=true
    bUseGravityOverride=true
    linearSpringConstant=10000
    simulationSpace=World
    alphaBoolBlend {blendOption=Linear}
    alphaScaleBiasClamp {bInterpResult=true, interpSpeedDecreasing=20, interpSpeedIncreasing=20}
    boundBone {boneName=dyn_wing_middle_base_01}
    componentAppliedLinearAccClamp=<100 100 100>
    componentLinearAccScale=<0.1 0.1 0.1>
    componentLinearVelScale=<0.2 0.2 0.2>
    gravityOverride {z=-600}
    physicsBodyDefinitions [{
      boundBone {boneName=dyn_wing_middle_base_01}
      boxExtents=<30 40 40>
      constraintSetup {angularLimitsMax=<0 8 3>, angularLimitsMin=<0 -10 -3>, angularTarget=<0 0 0>}
      localJointOffset=<30 0 0>
     }]
    relativeSpaceBone {boneName=Melee_TwoHanded_Handle}
   }
   n8 = Anim Dynamics - Bone: dyn_wing_lower_base_01(ComponentPose=@n7, Alpha=@n1, ExternalForce="0,0,0") {
    angularBiasOverride=1
    angularDampingOverride=1
    angularSpringConstant=10000
    bAngularSpring=true
    bOverrideAngularBias=true
    bOverrideAngularDamping=true
    bUseGravityOverride=true
    linearSpringConstant=10000
    simulationSpace=World
    alphaBoolBlend {blendOption=Linear}
    alphaScaleBiasClamp {bInterpResult=true, interpSpeedDecreasing=20, interpSpeedIncreasing=20}
    boundBone {boneName=dyn_wing_lower_base_01}
    componentAppliedLinearAccClamp=<100 100 100>
    componentLinearAccScale=<0.1 0.1 0.1>
    componentLinearVelScale=<0.2 0.2 0.2>
    gravityOverride {z=-600}
    physicsBodyDefinitions [{
      boundBone {boneName=dyn_wing_lower_base_01}
      boxExtents=<30 40 40>
      constraintSetup {angularLimitsMax=<0 0 3>, angularLimitsMin=<0 -2 -3>, angularTarget=<0 0 0>}
      localJointOffset=<30 0 0>
     }]
    relativeSpaceBone {boneName=Melee_TwoHanded_Handle}
   }
   n9 = Anim Dynamics - Bone: dyn_wing_upper_01(ComponentPose=@n8, Alpha=0.5, ExternalForce="0,0,0") {
    angularBiasOverride=1
    angularDampingOverride=1
    angularSpringConstant=10000
    bAngularSpring=true
    bOverrideAngularBias=true
    bOverrideAngularDamping=true
    bUseGravityOverride=true
    linearSpringConstant=10000
    simulationSpace=World
    alphaBoolBlend {blendOption=Linear}
    alphaScaleBiasClamp {interpSpeedDecreasing=20, interpSpeedIncreasing=20}
    boundBone {boneName=dyn_wing_upper_01}
    componentAppliedLinearAccClamp=<100 100 100>
    componentLinearAccScale=<0.1 0.1 0.1>
    componentLinearVelScale=<0.2 0.2 0.2>
    gravityOverride {z=-600}
    physicsBodyDefinitions [{
      boundBone {boneName=dyn_wing_upper_01}
      boxExtents=<10 10 10>
      constraintSetup {angularLimitsMax=<0 9 0>, angularLimitsMin=<0 -6 0>, angularTarget=<0 0 0>}
      localJointOffset=<2 0 0>
     }]
    relativeSpaceBone {boneName=Melee_TwoHanded_Handle}
   }
   n10 = Anim Dynamics - Bone: dyn_wing_middle_01(ComponentPose=@n9, Alpha=0.5, ExternalForce="0,0,0") {
    angularBiasOverride=1
    angularDampingOverride=1
    angularSpringConstant=10000
    bAngularSpring=true
    bOverrideAngularBias=true
    bOverrideAngularDamping=true
    bUseGravityOverride=true
    linearSpringConstant=10000
    simulationSpace=World
    alphaBoolBlend {blendOption=Linear}
    alphaScaleBiasClamp {interpSpeedDecreasing=20, interpSpeedIncreasing=20}
    boundBone {boneName=dyn_wing_middle_01}
    componentAppliedLinearAccClamp=<100 100 100>
    componentLinearAccScale=<0.1 0.1 0.1>
    componentLinearVelScale=<0.2 0.2 0.2>
    gravityOverride {z=-600}
    physicsBodyDefinitions [{
      boundBone {boneName=dyn_wing_middle_01}
      boxExtents=<10 10 10>
      constraintSetup {angularLimitsMax=<0 9 0>, angularLimitsMin=<0 -6 0>, angularTarget=<0 0 0>}
      localJointOffset=<2 0 0>
     }]
    relativeSpaceBone {boneName=Melee_TwoHanded_Handle}
   }
   n11 = Anim Dynamics - Bone: dyn_wing_lower_01(ComponentPose=@n10, Alpha=0.5, ExternalForce="0,0,0") {
    angularBiasOverride=1
    angularDampingOverride=1
    angularSpringConstant=10000
    bAngularSpring=true
    bOverrideAngularBias=true
    bOverrideAngularDamping=true
    bUseGravityOverride=true
    linearSpringConstant=10000
    simulationSpace=World
    alphaBoolBlend {blendOption=Linear}
    alphaScaleBiasClamp {interpSpeedDecreasing=20, interpSpeedIncreasing=20}
    boundBone {boneName=dyn_wing_lower_01}
    componentAppliedLinearAccClamp=<100 100 100>
    componentLinearAccScale=<0.1 0.1 0.1>
    componentLinearVelScale=<0.2 0.2 0.2>
    gravityOverride {z=-600}
    physicsBodyDefinitions [{
      boundBone {boneName=dyn_wing_lower_01}
      boxExtents=<10 10 10>
      constraintSetup {angularLimitsMax=<0 9 0>, angularLimitsMin=<0 -6 0>, angularTarget=<0 0 0>}
      localJointOffset=<2 0 0>
     }]
    relativeSpaceBone {boneName=Melee_TwoHanded_Handle}
   }
   n12 = Component To Local(ComponentPose=@n11)
   n13 = Get bAnimDynamicsEnabled() -> (bAnimDynamicsEnabled:bool)
   n14 = Use cached pose 'Post-CR'()
   n15 = RigidBody(ComponentPose=Local To Component(LocalPose=Use cached pose 'Post-CR'()), Alpha=0.4) {
    bOverrideWorldGravity=true
    overridePhysicsAsset=&/Lazuli_Axe_Physics
    simulationSpace=BaseBoneSpace
    alphaBoolBlend {blendOption=Linear}
    baseBoneRef {boneName=Melee_TwoHanded_Handle}
    componentAppliedLinearAccClamp=<3000 3000 3000>
    componentLinearAccScale=<0.3 0.3 0.3>
    componentLinearVelScale=<0.3 0.3 0.3>
    overrideWorldGravity {z=-980}
   }
   Output Pose(Result=Blend Poses by bool(
     BlendPose_0=Component To Local(ComponentPose=@n15)
     BlendPose_1=Blend Poses by bool(BlendPose_0=@n12, BlendPose_1=@n14, bActiveValue=@n13) {blendType=Linear}
     bActiveValue=Get bRigidBodyEnabled()
    ) {blendType=Linear}) {layerGroup=None}
   n16 = Control Rig(Source=Input Pose(), Alpha=Get ControlRigAlpha() -> (ControlRigAlpha:double)) {
    alphaBoolBlend {blendOption=Linear}
    alphaScaleBiasClamp {bInterpResult=true, interpSpeedDecreasing=3, interpSpeedIncreasing=3}
    controlRigAssetReference {blueprintRigClass=&/Lazuli_Axe_CtrlRig.Lazuli_Axe_CtrlRig_C}
   }
   Save cached pose 'Post-CR'(Pose=@n16) {cacheName=Post-CR}
  ]}
 EventGraph {nodes [
   n1 = Event BlueprintUpdateAnimation()
   > Set Delta Time X(Delta Time X=@n1.DeltaTimeX)
   n2 = Get PawnSpeed2D()
   n3 = "float > float"(A=@n2, B=1) {is=title K2Node_CallFunction}
   n4 = Get bIsJumping()
   n5 = Get bIsFalling()
   > Set ControlRigAlpha(ControlRigAlpha:double=SelectFloat(A=0, B=1, bPickA=OR Boolean!(A=@n4, B=@n5, C:bool=@n3) -> (ReturnValue))) -> (Output_Get:double)
   n6 = "float <= float"(A=@n2) -> (ReturnValue=~)
   n7 = Get bIsPlayingMeleeAnim()
   > Set AnimDynamicsAlpha(AnimDynamicsAlpha:double=SelectFloat(A=0.4, B=0, bPickA=OR Boolean!(A=@n4, B=@n5, C=@n7, D:bool=@n3, E:bool=@n6) -> (ReturnValue))) -> (Output_Get:double)
   n8 = OR Boolean!(A=Get bIsJumping(), B=Get bIsFalling(), C=Get bIsSliding()) -> (ReturnValue)
   > Set WingTransformWhileJumpingAlpha(WingTransformWhileJumpingAlpha:double=SelectFloat(A=1, B=0, bPickA=@n8)) -> (Output_Get:double)
   > n9 = Branch(Condition=@n8)
   > Set WingTransformWhileJumpingAlpha(WingTransformWhileJumpingAlpha=FInterpTo(Current=Get WingTransformWhileJumpingAlpha(), Target=1, DeltaTime=Get Delta Time X(), InterpSpeed=10))
   Set WingTransformWhileJumpingAlpha(execute=@n9.else, WingTransformWhileJumpingAlpha=0)
   TryGetPawnOwner()
  ]}
]
variables [
 AnimDynamicsAlpha {type=number}
 ControlRigAlpha {type=number}
 Delta Time X {type=number}
 WingTransformWhileJumpingAlpha {type=number}
]
